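# Physics asset for the HIS face accessory mesh of the female medium Slither player body, covering its hair bones.
# Small capsules on six hair bones form three two-bone chains (A1-A2, B1-B2, C3-C2), each constrained to the head, which has a zero-length capsule acting as a sphere.
# Constraints allow narrow swing with twist limits at zero, keeping hair motion tight.
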PhysicsAsset /BRCosmetics/Characters/Player/Female/Medium/Bodies/F_MED_Slither/Meshes/Parts/HIS/F_MED_Slither_HIS_FaceAcc_Physics1
bodies {
 HairA1 [{shapeType=Capsule, length=1.75, radius=1.75, shapeName=hairA2_capsule, center=<2 0 2.5>, rotation=<rot 90 0 0>}]
 HairA2 [{shapeType=Capsule, length=1.75, radius=1.75, shapeName=hairA2_capsule, center=<2 0 1.75>, rotation=<rot 72.372 0 0>}]
 HairB1 [{shapeType=Capsule, length=1.75, radius=1.75, shapeName=hairA2_capsule, center=<2 0 2.5>, rotation=<rot 90 0 0>}]
 HairB2 [{shapeType=Capsule, length=1.75, radius=1.75, shapeName=hairA2_capsule, center=<1.944 0 2>, rotation=<rot 75.3234 0 0>}]
 HairC2 [{shapeType=Capsule, length=1.75, radius=1.75, shapeName=hairA2_capsule, center=<1.7421 0 1.3>, rotation=<rot 64.9963 0 0>}]
 HairC3 [{shapeType=Capsule, length=1.75, radius=1.75, shapeName=hairA2_capsule, center=<1.9916 0 2>, rotation=<rot 84.4293 0 0>}]
 head [{shapeType=Capsule, length=0, radius=8.3681, shapeName=head_capsule, center=<5.1648 0.504 0.0002>}]
}
constraints [
 {bone1Name=HairA2, bone2Name=HairA1, swing1LimitDegrees=15, swing2LimitDegrees=20, twistLimitDegrees=0}
 {bone1Name=HairC2, bone2Name=HairC3, swing1LimitDegrees=10, swing2LimitDegrees=20, twistLimitDegrees=0}
 {bone1Name=HairB2, bone2Name=HairB1, swing1LimitDegrees=10, swing2LimitDegrees=20, twistLimitDegrees=0}
 {bone1Name=HairA1, bone2Name=head, swing1LimitDegrees=15, swing2LimitDegrees=20, twistLimitDegrees=0}
 {bone1Name=HairC3, bone2Name=head, swing1LimitDegrees=10, swing2LimitDegrees=20, twistLimitDegrees=0}
 {bone1Name=HairB1, bone2Name=head, swing1LimitDegrees=10, swing2LimitDegrees=20, twistLimitDegrees=0}
]
