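# Physics asset for the headband face accessory on the Slither female medium character, named as the skydive variant.
# Two capsule chains hang from headbandMain, five links right and four left, with tight twist limits and wider swing toward the tips.
# Capsules on head, neck and upper spine give the body volume, joined to each other and to headbandMain by angularly free constraints.
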PhysicsAsset /BRCosmetics/Characters/Player/Female/Medium/Bodies/F_MED_Slither/Meshes/Parts/HIS/F_MED_Slither_HIS_FaceAcc_PhysicsHeadbandSkydive
bodies {
 head [{shapeType=Capsule, length=0, radius=8.3681, shapeName=head_capsule, center=<5.1648 0.504 0.0002>}]
 headbandLf1 [{shapeType=Capsule, length=1, radius=1.5, shapeName=headbandRt1_capsule, center=<2 -0.5 0>, rotation=<rot 90 0 0>}]
 headbandLf2 [{shapeType=Capsule, length=1, radius=1.5, shapeName=headbandRt1_capsule, center=<2 -0.5 0>, rotation=<rot 90 0 0>}]
 headbandLf3 [{shapeType=Capsule, length=1, radius=1.5, shapeName=headbandRt1_capsule, center=<2 -0.5 0>, rotation=<rot 90 0 0>}]
 headbandLf4 [{shapeType=Capsule, length=2.25, radius=1.5, shapeName=headbandRt1_capsule, center=<2.9005 -0.5001 0>, rotation=<rot 90 0 0>}]
 headbandMain [{shapeType=Capsule, length=0.505, radius=0.505, shapeName=headbandMain_capsule, rotation=<rot 0 -90 0>}]
 headbandRt1 [{shapeType=Capsule, length=1, radius=1.5, shapeName=headbandRt1_capsule, center=<2 -0.5 0>, rotation=<rot 90 0 0>}]
 headbandRt2 [{shapeType=Capsule, length=1, radius=1.5, shapeName=headbandRt1_capsule, center=<2 -0.5 0>, rotation=<rot 90 0 0>}]
 headbandRt3 [{shapeType=Capsule, length=1, radius=1.5, shapeName=headbandRt1_capsule, center=<2 -0.5 0>, rotation=<rot 90 0 0>}]
 headbandRt4 [{shapeType=Capsule, length=1, radius=1.5, shapeName=headbandRt1_capsule, center=<2 -0.5 0>, rotation=<rot 90 0 0>}]
 headbandRt5 [{shapeType=Capsule, length=2.25, radius=1.5, shapeName=headbandRt1_capsule, center=<2.9005 -0.5001 0>, rotation=<rot 90 0 0>}]
 neck_01 [{shapeType=Capsule, length=12.5, radius=5.255, shapeName=neck_02_capsule, center=<2.5525 0 0>, rotation=<rot 90 0 0>}]
 neck_02 [{shapeType=Capsule, length=2, radius=4.005, shapeName=neck_02_capsule, center=<2.5525 0 0>, rotation=<rot 90 0 0>}]
 spine_05 [
  {shapeType=Capsule, length=16.005, radius=10.005, shapeName=spine_05_capsule, center=<3.5675 1.0616 0>, rotation=<rot 0 -89.9999 0>}
  {shapeType=Capsule, length=15, radius=15, shapeName=spine_05_capsule, center=<6.2318 12.9193 -0.0003>}
  {shapeType=Capsule, length=0, radius=8.5, shapeName=spine_05_capsule, center=<7.731 0 0>}
 ]
}
constraints [
 {bone1Name=headbandRt1, bone2Name=headbandMain, swing1LimitDegrees=20, swing2LimitDegrees=10, twistLimitDegrees=0}
 {bone1Name=headbandRt2, bone2Name=headbandRt1, swing1LimitDegrees=20, swing2LimitDegrees=10, twistLimitDegrees=5}
 {bone1Name=headbandRt3, bone2Name=headbandRt2, swing1LimitDegrees=20, swing2LimitDegrees=10, twistLimitDegrees=5}
 {bone1Name=headbandRt4, bone2Name=headbandRt3, swing1LimitDegrees=45, swing2LimitDegrees=10, twistLimitDegrees=5}
 {bone1Name=headbandRt5, bone2Name=headbandRt4, swing1LimitDegrees=45, swing2LimitDegrees=10, twistLimitDegrees=5}
 {bone1Name=headbandLf1, bone2Name=headbandMain, swing1LimitDegrees=20, swing2LimitDegrees=10, twistLimitDegrees=0}
 {bone1Name=headbandLf2, bone2Name=headbandLf1, swing1LimitDegrees=20, swing2LimitDegrees=10, twistLimitDegrees=5}
 {bone1Name=headbandLf3, bone2Name=headbandLf2, swing1LimitDegrees=45, swing2LimitDegrees=10, twistLimitDegrees=5}
 {bone1Name=headbandLf4, bone2Name=headbandLf3, swing1LimitDegrees=45, swing2LimitDegrees=10, twistLimitDegrees=5}
 {
  bone1Name=headbandMain
  bone2Name=head
  swing1LimitDegrees=45
  swing1Motion=Free
  swing2LimitDegrees=45
  swing2Motion=Free
  twistLimitDegrees=45
  twistMotion=Free
 }
 {
  bone1Name=neck_01
  bone2Name=spine_05
  swing1LimitDegrees=45
  swing1Motion=Free
  swing2LimitDegrees=45
  swing2Motion=Free
  twistLimitDegrees=45
  twistMotion=Free
 }
 {
  bone1Name=neck_02
  bone2Name=neck_01
  swing1LimitDegrees=45
  swing1Motion=Free
  swing2LimitDegrees=45
  swing2Motion=Free
  twistLimitDegrees=45
  twistMotion=Free
 }
 {
  bone1Name=head
  bone2Name=neck_02
  swing1LimitDegrees=45
  swing1Motion=Free
  swing2LimitDegrees=45
  swing2Motion=Free
  twistLimitDegrees=45
  twistMotion=Free
 }
]
